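# Pivot Painter 2 foliage material function that adds four-level hierarchical wind animation to incoming material attributes.
# It unpacks parent indices from the position texture up to three ancestors deep, using hierarchy depth to pick UVs and layer masks.
# Four PivotPainter2FoliageAnimation calls chain world-space normals, optionally pass rotation to deeper levels, and sum World Position Offset.
# Only SM5/SM6 get the animated attributes; other feature levels get an unconnected MakeMaterialAttributes.
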MaterialFunction /Engine/Functions/Engine_MaterialFunctions02/PivotPainter2/PivotPainter2FoliageShader
graph {
 properties {
  bExposeToLibrary=true
  description="This material function contains texture and numeric parameters that should be altered to fit your particular asset. It is suggested that this be done at the material instance level. "
  libraryCategoriesText [Pivot Painter 2]
 }
 nodes [
  n1 = MakeMaterialAttributes!(
   BaseColor
   Metallic
   Specular
   Roughness
   Anisotropy
   EmissiveColor
   Opacity
   OpacityMask
   Normal
   Tangent
   WorldPositionOffset
   SubsurfaceColor
   ClearCoat
   ClearCoatRoughness
   AmbientOcclusion
   Refraction
   CustomizedUVs_0
   CustomizedUVs_1
   CustomizedUVs_2
   CustomizedUVs_3
   CustomizedUVs_4
   CustomizedUVs_5
   CustomizedUVs_6
   CustomizedUVs_7
   PixelDepthOffset
   ShadingModel
   Displacement
  )
  n2 = FunctionInput(Preview=MakeMaterialAttributes()) {
   bUsePreviewValueAsDefault=true
   description=Please make sure that you are using tangent space normals within your material and that the use tangent space normals option is unchecked. The normals will be transformed to world space internally.
   inputName="Material Attributes (See Note)"
   inputType=FunctionInput_MaterialAttributes
  }
  n3 = BreakMaterialAttributes!(Attr=@n2)
  n4 = TextureObjectParameter() {
   group=Shared Wind Settings
   parameterName=Position and Index Texture
   samplerType=SAMPLERTYPE_LinearColor
   texture=/Engine/Functions/Engine_MaterialFunctions02/ExampleContent/PivotPainter2/SimpExPivPos
  }
  n5 = TextureObjectParameter() {
   group=Shared Wind Settings
   parameterName=X-Vector And X-Extent Texture
   samplerType=SAMPLERTYPE_LinearColor
   texture=/Engine/Functions/Engine_MaterialFunctions02/ExampleContent/PivotPainter2/PP2_rgb_XVect_a_XExtent
  }
  n6 = FunctionInput(Preview=TextureCoordinate() {coordinateIndex=1}) {
   bUsePreviewValueAsDefault=true
   description=Defaults to UV Coordinate 1.
   inputName=Pivot Painter Texture Coordinate
   inputType=FunctionInput_Vector2
   sortPriority=1
  }
  n7 = TextureSampleParameter2D(UVs=@n6) {
   group=Shared Wind Settings
   parameterName=Position and Index Texture
   samplerType=SAMPLERTYPE_LinearColor
   texture=/Engine/Functions/Engine_MaterialFunctions02/ExampleContent/PivotPainter2/test2_rgb_PivotPos_a_ParentIndexInt_UV_2
  }
  n8 = MaterialFunctionCall!(Integer as Float=@n7.A) {materialFunction=/Engine/Functions/Engine_MaterialFunctions02/ms_PivotPainter2_UnpackIntegerAsFloat}
  n9 = TextureObjectParameter() {
   group=Shared Wind Settings
   parameterName=Position and Index Texture
   samplerType=SAMPLERTYPE_LinearColor
   texture=&SimpExPivPos
  }
  n10 = TextureProperty(Tex=@n9)
  n11 = MaterialFunctionCall!(
   "Parent Index As Float (See note)"=@n8.Result
   Texture Dimensions=@n10
   "Current Index (see note)"=MaterialFunctionCall!(Data Texture Dimensions=@n10, Pivot Painter UV Coordinates=@n6) {materialFunction=&/ms_PivotPainter2_CalculateMeshElementIndex}.Index
  ) {materialFunction=&/ms_PivotPainter2_ReturnParentTextureInfo}
  n12 = TextureSampleParameter2D(UVs=@n11.Parent UVs) {
   group=Shared Wind Settings
   parameterName=Position and Index Texture
   samplerType=SAMPLERTYPE_LinearColor
   texture=&test2_rgb_PivotPos_a_ParentIndexInt_UV_2
  }
  n13 = MaterialFunctionCall(Integer as Float=@n12.A) {materialFunction=&ms_PivotPainter2_UnpackIntegerAsFloat}
  n14 = MaterialFunctionCall("Parent Index As Float (See note)"=@n13.Result, Texture Dimensions=@n10, "Current Index (see note)"=@n8.Result) {materialFunction=&ms_PivotPainter2_ReturnParentTextureInfo}
  n15 = TextureSampleParameter2D(UVs=@n14.Parent UVs) {
   group=Shared Wind Settings
   parameterName=Position and Index Texture
   samplerType=SAMPLERTYPE_LinearColor
   texture=&test2_rgb_PivotPos_a_ParentIndexInt_UV_2
  }
  n16 = MaterialFunctionCall(
   "Parent Index As Float (See note)"=MaterialFunctionCall(Integer as Float=@n15.A) {materialFunction=&ms_PivotPainter2_UnpackIntegerAsFloat}.Result
   Texture Dimensions=@n10
   "Current Index (see note)"=@n13.Result
  ) {materialFunction=&ms_PivotPainter2_ReturnParentTextureInfo}
  n17 = Ceil(None=Add(A=Add(A=@n14."Is Child? (See note)", B=@n11."Is Child? (See note)"), B=@n16."Is Child? (See note)"))
  n18 = TextureObjectParameter() {
   group=Shared Wind Settings
   parameterName=Position and Index Texture
   samplerType=SAMPLERTYPE_LinearColor
   texture=&SimpExPivPos
  }
  n19 = TextureProperty(Tex=@n18)
  n20 = MaterialFunctionCall!(
   Enable=StaticBoolParameter() {group=Wind Setting 1, parameterName=Animate Level 1}
   Pivot Painter Texture Coordinate=MaterialFunctionCall!(1=@n6, 2=@n11.Parent UVs, 3=@n14.Parent UVs, 4=@n16.Parent UVs, A=Divide(A=@n17) {constB=3}) {materialFunction=/Engine/Functions/Engine_MaterialFunctions02/Lerp_Multiple_Float2}.Lerp 4 Inputs
   Enable Normal Rotation=StaticBoolParameter() {group=Wind Setting 1, parameterName=Animate Level 1 Normals}
   Motion Dampening Falloff Radius=ScalarParameter() {defaultValue=2, group=Wind Setting 1, parameterName=Dampening Radius Multiplier 1, sliderMin=0.01}
   Wind Speed Down Wind Vector=Multiply(
    A=ScalarParameter() {defaultValue=1, group=Wind Setting 1, parameterName=Local Wind Scale 1}
    B=ScalarParameter() {defaultValue=0.25, group=Shared Wind Settings, parameterName=Wind Speed}
   )
   Wind Gust Angle Rotation=ScalarParameter() {defaultValue=0.0092, group=Wind Setting 1, parameterName=Max Rotation 1, sliderMax=10}
   Wind Gust Offset=ScalarParameter() {group=Wind Setting 1, parameterName=Wind Angle Offset 1}
   Optional - world space pixel normal=Transform(None=BreakMaterialAttributes(Attr=@n2).Normal)
   Texture Dimensions To Gather Parent UVs=@n19
   Layer Mask=Constant() {r=1}
   Wind Shelter Settings=AppendVector(
    A=ScalarParameter() {defaultValue=-1.25, group=Wind Setting 1, parameterName="Wind 1 Shelter Dot Product Bias "}
    B=ScalarParameter() {defaultValue=-0.444, group=Wind Setting 1, parameterName="Wind 1 Shelter Dot Product Modulation "}
   )
   Advanced - Parent Rotation Angle Influence=Constant()
   Random Rotation Texture Sample Scale=ScalarParameter() {defaultValue=200, group=Wind Setting 1, parameterName=Wind 1 Random Rotation Change Rate}
   Random Rotation Influence=ScalarParameter() {defaultValue=0.25, group=Wind Setting 1, parameterName=Wind 1 Random Rotation Influence}
   Wind Direction
   Wind Direction Y Axis
   Position And Index Texture=@n4
   X-axis and X-Extent Texture=@n5
   Wind Gust World Scale
   Wind Speed Horizontal Speed
   Optional - Wind Turbulence and Gust Mag
   Use Wind Shelter Settings
  ) {materialFunction=&/PivotPainter2FoliageAnimation}
  n21 = Constant()
  n22 = StaticSwitchParameter(
   True=Multiply(A=@n20.Advanced - Rotation Angle Animation, B=ScalarParameter() {group=Wind Setting 4, parameterName=Wind 1's Angle Contribution to Wind 4})
   False=@n21
  ) {group=Wind Setting 4, parameterName=Allow Wind 1 Calculations To Affect Wind 4's Rotation}
  n23 = Subtract(A=@n17, B=Constant() {r=1})
  n24 = Divide(A=@n23)
  n25 = ScalarParameter() {group=Wind Setting 2, parameterName=Wind 1's Angle Contribution to Wind 2}
  n26 = MaterialFunctionCall(
   Enable=StaticBoolParameter() {group=Wind Setting 2, parameterName=Animate Level 2}
   Pivot Painter Texture Coordinate=MaterialFunctionCall(1=@n6, 2=@n11.Parent UVs, 3=@n14.Parent UVs, 4=@n16.Parent UVs, A=Clamp(None=@n24)) {materialFunction=&Lerp_Multiple_Float2}.Lerp 3 Inputs
   Enable Normal Rotation=StaticBoolParameter() {group=Wind Setting 2, parameterName=Animate Level 2 Normals}
   Motion Dampening Falloff Radius=ScalarParameter() {defaultValue=2, group=Wind Setting 2, parameterName=Dampening Radius Multiplier 2, sliderMin=0.01}
   Wind Speed Down Wind Vector=Multiply(
    A=ScalarParameter() {defaultValue=1, group=Wind Setting 2, parameterName=Local Wind Scale 2}
    B=ScalarParameter() {defaultValue=0.25, group=Shared Wind Settings, parameterName=Wind Speed}
   )
   Wind Gust Angle Rotation=ScalarParameter() {defaultValue=0.0092, group=Wind Setting 2, parameterName=Max Rotation 2, sliderMax=10}
   Wind Gust Offset=ScalarParameter() {group=Wind Setting 2, parameterName=Wind Angle offset 2}
   Optional - world space pixel normal=@n20.Rotated World Space Pixel Normal
   Texture Dimensions To Gather Parent UVs=@n19
   Layer Mask=Clamp(None=@n17)
   Wind Shelter Settings=AppendVector(
    A=ScalarParameter() {defaultValue=-1.25, group=Wind Setting 2, parameterName="Wind 2 Shelter Dot Product Bias "}
    B=ScalarParameter() {defaultValue=-0.444, group=Wind Setting 2, parameterName="Wind 2 Shelter Dot Product Modulation "}
   )
   Advanced - Parent Rotation Angle Influence=StaticSwitchParameter(True=Multiply(A=@n20.Advanced - Rotation Angle Animation, B=@n25), False=Constant()) {group=Wind Setting 2, parameterName=Allow Wind 1 Calculations To Affect Wind 2's Rotation}
   Random Rotation Texture Sample Scale=ScalarParameter() {defaultValue=200, group=Wind Setting 2, parameterName=Wind 2 Random Rotation Change Rate}
   Random Rotation Influence=ScalarParameter() {defaultValue=0.25, group=Wind Setting 2, parameterName=Wind 2 Random Rotation Influence}
  ) {materialFunction=&PivotPainter2FoliageAnimation}
  n27 = StaticSwitchParameter(
   True=Multiply(A=@n26.Advanced - Rotation Angle Animation, B=ScalarParameter() {group=Wind Setting 4, parameterName=Wind 2's Angle Contribution to Wind 4})
   False=@n21
  ) {group=Wind Setting 4, parameterName=Allow Wind 2 Calculations To Affect Wind 4's Rotation}
  n28 = Multiply(A=@n20.Advanced - Rotation Angle Animation, B=ScalarParameter() {group=Wind Setting 3, parameterName=Wind 1's Angle Contribution to Wind 3})
  n29 = Constant()
  n30 = Multiply(A=@n26.Advanced - Rotation Angle Animation, B=ScalarParameter() {group=Wind Setting 3, parameterName=Wind 2's Angle Contribution to Wind 3})
  n31 = Clamp(None=Subtract(A=@n17, B=Constant() {r=2}))
  n32 = MaterialFunctionCall(
   Enable=StaticBoolParameter() {group=Wind Setting 3, parameterName=Animate Level 3}
   Pivot Painter Texture Coordinate=LinearInterpolate(A=@n6, B=@n11.Parent UVs, Alpha=@n31)
   Enable Normal Rotation=StaticBoolParameter() {group=Wind Setting 3, parameterName=Animate Level 3 Normals}
   Motion Dampening Falloff Radius=ScalarParameter() {defaultValue=2, group=Wind Setting 3, parameterName=Dampening Radius Multiplier 3, sliderMin=0.01}
   Wind Speed Down Wind Vector=Multiply(
    A=ScalarParameter() {defaultValue=1, group=Wind Setting 3, parameterName=Local Wind Scale 3}
    B=ScalarParameter() {defaultValue=0.25, group=Shared Wind Settings, parameterName=Wind Speed}
   )
   Wind Gust Angle Rotation=ScalarParameter() {defaultValue=0.0092, group=Wind Setting 3, parameterName=Max Rotation 3, sliderMax=10}
   Wind Gust Offset=ScalarParameter() {group=Wind Setting 3, parameterName=Wind Angle Offset 3}
   Optional - world space pixel normal=@n26.Rotated World Space Pixel Normal
   Texture Dimensions To Gather Parent UVs=@n19
   Layer Mask=Clamp(None=@n23)
   Wind Shelter Settings=AppendVector(
    A=ScalarParameter() {defaultValue=-1.25, group=Wind Setting 3, parameterName="Wind 3 Shelter Dot Product Bias "}
    B=ScalarParameter() {defaultValue=-0.444, group=Wind Setting 3, parameterName="Wind 3 Shelter Dot Product Modulation "}
   )
   Advanced - Parent Rotation Angle Influence=Add(
    A=StaticSwitchParameter(True=@n28, False=@n29) {group=Wind Setting 3, parameterName=Allow Wind 1 Calculations To Affect Wind 3's Rotation}
    B=StaticSwitchParameter(True=@n30, False=@n29) {group=Wind Setting 3, parameterName=Allow Wind 2 Calculations To Affect Wind 3's Rotation}
   )
   Random Rotation Texture Sample Scale=ScalarParameter() {defaultValue=200, group=Wind Setting 3, parameterName=Wind 3 Random Rotation Change Rate}
   Random Rotation Influence=ScalarParameter() {defaultValue=0.25, group=Wind Setting 3, parameterName=Wind 3 Random Rotation Influence}
  ) {materialFunction=&PivotPainter2FoliageAnimation}
  n33 = Multiply(
   A=@n32.Advanced - Rotation Angle Animation
   B=ScalarParameter() {defaultValue=1, group=Wind Setting 4, parameterName=Wind 3's Angle Contribution to Wind 4}
  )
  n34 = MaterialFunctionCall(
   Enable=StaticBoolParameter() {group=Wind Setting 4, parameterName=Animate Level 4}
   Pivot Painter Texture Coordinate=@n6
   Enable Normal Rotation=StaticBoolParameter() {group=Wind Setting 4, parameterName=Animate Level 4 Normals}
   Motion Dampening Falloff Radius=ScalarParameter() {defaultValue=2, group=Wind Setting 4, parameterName=Dampening Radius Multiplier 4, sliderMin=0.01}
   Wind Speed Down Wind Vector=Multiply(
    A=ScalarParameter() {defaultValue=0.25, group=Shared Wind Settings, parameterName=Wind Speed}
    B=ScalarParameter() {defaultValue=2, group=Wind Setting 4, parameterName=Local Wind 4 Speed}
   )
   Wind Gust Angle Rotation=ScalarParameter() {defaultValue=0.0092, group=Wind Setting 4, parameterName=Max Rotation 4, sliderMax=10}
   Wind Gust Offset=ScalarParameter() {group=Wind Setting 4, parameterName=Wind Angle Offset 4}
   Optional - world space pixel normal=@n32.Rotated World Space Pixel Normal
   Texture Dimensions To Gather Parent UVs=@n19
   Layer Mask=@n31
   Wind Shelter Settings=AppendVector(
    A=ScalarParameter() {defaultValue=-1.25, group=Wind Setting 4, parameterName="Wind 4 Shelter Dot Product Bias "}
    B=ScalarParameter() {defaultValue=-0.444, group=Wind Setting 4, parameterName="Wind 4 Shelter Dot Product Modulation "}
   )
   Advanced - Parent Rotation Angle Influence=Add(
    A=Add(A=@n22, B=@n27)
    B=StaticSwitchParameter(True=@n33, False=@n21) {group=Wind Setting 4, parameterName=Allow Wind 2 Calculations To Affect Wind 4's Rotation}
   )
   Random Rotation Texture Sample Scale=ScalarParameter() {defaultValue=200, group=Wind Setting 4, parameterName=Wind 4 Random Rotation Change Rate}
   Random Rotation Influence=ScalarParameter() {defaultValue=0.25, group=Wind Setting 4, parameterName=Wind 4 Random Rotation Influence}
  ) {materialFunction=&PivotPainter2FoliageAnimation}
  n35 = Add(A=Add(A=@n20.World Position Offset, B=@n26.World Position Offset), B=@n32.World Position Offset)
  n36 = MakeMaterialAttributes(
   BaseColor=@n3.BaseColor
   Metallic=@n3.BaseColor
   Specular=@n3.BaseColor
   Roughness=@n3.BaseColor
   EmissiveColor=@n3.BaseColor
   Opacity=@n3.BaseColor
   OpacityMask=@n3.BaseColor
   Normal=FunctionOutput(None=@n34.Rotated World Space Pixel Normal) {outputName=Modified World Space Normal Component, sortPriority=1}
   WorldPositionOffset=FunctionOutput(None=Add(A=@n35, B=@n34.World Position Offset)) {outputName=World Position Offset Component, sortPriority=2}
   SubsurfaceColor=@n3.BaseColor
   ClearCoat=@n3.BaseColor
   ClearCoatRoughness=@n3.BaseColor
   AmbientOcclusion=@n3.BaseColor
   Refraction=@n3.BaseColor
   CustomizedUVs_0=@n3.BaseColor
   CustomizedUVs_1=@n3.BaseColor
   CustomizedUVs_2=@n3.BaseColor
   CustomizedUVs_3=@n3.BaseColor
   CustomizedUVs_4=@n3.BaseColor
   CustomizedUVs_5=@n3.BaseColor
   CustomizedUVs_6=@n3.BaseColor
   CustomizedUVs_7=@n3.BaseColor
   PixelDepthOffset=@n3.BaseColor
  )
  FunctionOutput(None=FeatureLevelSwitch(Default=@n1, ES2=@n1, ES3_1=@n1, SM4_REMOVED=@n1, SM5=@n36, SM6=@n36) {desc=Only supported in SM5 for now}) {outputName=Final Material with World Space Normals}
 ]
}
