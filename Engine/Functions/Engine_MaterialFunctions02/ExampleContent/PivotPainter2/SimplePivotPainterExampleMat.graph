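# Minimal example material for Pivot Painter 2, showing how the PivotPainter2FoliageShader material function wraps a simple lit, metallic surface.
# A MakeMaterialAttributes node packs a BaseValue scalar base color, constant metallic and roughness, and a constant +Z normal with tangent-space normals disabled.
# Those attributes feed the foliage shader function, whose Pivot Painter texture coordinate input is left unconnected.
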Material /Engine/Functions/Engine_MaterialFunctions02/ExampleContent/PivotPainter2/SimplePivotPainterExampleMat
graph {
 properties {
  bTangentSpaceNormal=false
  bUseMaterialAttributes=true
  bUsedWithStaticLighting=true
  pixelDepthOffsetMode=PDOM_Legacy
  refractionMethod=RM_IndexOfRefraction
  usedShadingModels=Default Lit
 }
 nodes [MaterialFunctionCall!(
   "Material Attributes (See Note)"=MakeMaterialAttributes!(
    BaseColor=ScalarParameter() {defaultValue=0.2243, parameterName=BaseValue}
    Metallic=Constant() {r=1}
    Specular
    Roughness=Constant() {r=0.3486}
    Anisotropy
    EmissiveColor
    Opacity
    OpacityMask
    Normal=Constant3Vector() {constant=<rgba 0 0 1 0>}
    Tangent
    WorldPositionOffset
    SubsurfaceColor
    ClearCoat
    ClearCoatRoughness
    AmbientOcclusion
    Refraction
    CustomizedUVs_0
    CustomizedUVs_1
    CustomizedUVs_2
    CustomizedUVs_3
    CustomizedUVs_4
    CustomizedUVs_5
    CustomizedUVs_6
    CustomizedUVs_7
    PixelDepthOffset
    ShadingModel
    Displacement
   )
   Pivot Painter Texture Coordinate
  ) {materialFunction=/Engine/Functions/Engine_MaterialFunctions02/PivotPainter2/PivotPainter2FoliageShader}]
}
